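# Niagara effect for the TulipGlory emote that bursts dark ground cracks, drawn with one decal and three sprite renderers in a single emitter.
# Sprites and decal pop to size almost instantly and follow opacity curves scaled by User.OpacityMult, while User.WidthMult scales decal and glow widths.
# Camera offset and sprite opacity adapt to engine quality level and front-end display, and a visibility tag is set from the front-end preview flag.
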NiagaraSystem /BRCosmetics/Animation/Game/MainPlayer/Emotes/TulipGlory/FX/NS_Emote_TulipGlory_GroundCracks
dependencies {
 usedDataInterfaces [&NiagaraDataInterfaceArrayFloat, &NiagaraDataInterfaceCurve, &NiagaraDataInterfaceSkeletalMesh]
 usedDynamicInputs []
 usedModules [
  &self:SystemSpawnScript.SystemScriptSource.SystemScriptGraph.NiagaraNodeAssignment_0.SetVariables_#1
  &SystemState
 ]
}
emitters {Ground_Cracks {
  emitterUpdateScript [
   &EmitterState(
    Inactive Response=::NewEnumerator1
    Life Cycle Mode=::NewEnumerator1
    Loop Behavior=::NewEnumerator1
    Loop Delay=0.5
    Loop Duration Mode=::NewEnumerator1
    UseLoopDelay=false
   )
   &SpawnBurst_Instantaneous(Use Spawn Probability=false, UseLoopCountLimit=false)
  ]
  particleSpawnScript [
   &InitializeParticle(
    Color=<rgba 0.0104 0.0104 0.0104 1>
    Lifetime=4.5
    Sprite Rotation Mode=::NewEnumerator2
    Sprite Size Mode=::NewEnumerator1
    Uniform Sprite Size Max=130
    Uniform Sprite Size Min=90
    UsePositionOffset=true
    Write Material Random=true
   ) {module=InitializeParticle001}
   /SetVariables(Particles.RandomDecalYZ:NiagaraFloat=&Multiply_Float(A=&RandomRangeFloat(Minimum=60, Maximum=80), B=$User.WidthMult)) {isSetParameters=true}
   /SetVariables(Particles.DecalSize:Vector3f=&MakeVector(X=75, Y=$Particles.RandomDecalYZ, Z=$Particles.RandomDecalYZ)) {isSetParameters=true}
   &InitialMeshOrientation(
    Mesh Orientation Mode=::NewEnumerator2
    Orientation Vector=<0 0 1>
    Rotation=&RandomRangeVector(Minimum=<-1 0 0>, Maximum=<1 0 0>)
    Use Rotation Vector=true
   )
   /SetVariables(Particles.VisibilityTag=&MakeCustomIntFromBool(Bool=$User.bIsFrontEndPreview, True INT=0, False INT=1)) {isSetParameters=true}
   &SpriteFacingAndAlignment(Sprite Alignment=<1 0 0>, Sprite Facing=<0 0 1>, WriteAlignment=true, WriteFacing=true) {module=SpriteFacingAndAlignment001}
   &CameraOffset(Camera Offset Amount=&FN_FrontEndFloatModifier(
     Gameplay=&SelectFloatFromArray(Direct Array Index=$Engine.QualityLevel, Float Selection Array=DataInterface() {propertyValues {FloatData [5,0,0,0,0]}})
     Front End=10
    ))
  ]
  particleUpdateScript [
   &ParticleState(Kill Particles When Lifetime Has Elapsed=false, Let Infinitely Lived Particles Die When Emitter Deactivates=true)
   &ScaleSpriteSize(Uniform Curve Sprite Scale=DataInterface() {propertyValues {Curve {keys [
        {arriveTangent=230.8238, interpMode=RCIM_Cubic, leaveTangent=230.8238, tangentMode=RCTM_User, time=0, value=0}
        {arriveTangent=0.8135, interpMode=RCIM_Cubic, leaveTangent=0.8135, tangentMode=RCTM_User, time=0.01, value=1}
       ]}}})
   &ScaleColor(
    Scale Alpha=&FloatFromCurve(
     FloatCurve=DataInterface() {propertyValues {Curve {keys [
         {interpMode=RCIM_Cubic, tangentMode=RCTM_User, time=0, value=0}
         <at 0.025 1>
         {interpMode=RCIM_Cubic, leaveTangent=-1.0256, time=0.7, value=1}
        ]}}}
     Scale Curve=&Multiply_Float(
      A=$User.OpacityMult
      B=&FN_FrontEndFloatModifier(
       Gameplay=&SelectFloatFromArray(Direct Array Index=$Engine.QualityLevel, Float Selection Array=DataInterface() {propertyValues {FloatData [0.75,1.5,1.5,1.5,1.5]}})
       Front End=0.3
      )
     )
    )
    ScaleA=true
    ScaleRGB=true
   )
   &ScaleSpriteSize(Uniform Curve Sprite Scale=DataInterface() {propertyValues {Curve {keys [{arriveTangent=139.099, interpMode=RCIM_Cubic, leaveTangent=139.099, tangentMode=RCTM_User, time=0, value=0}, <at 0.015 1 RCIM_Cubic>]}}}) {module=ScaleSpriteSize001}
   /SetVariables(Particles.DecalFade:NiagaraFloat=&FloatFromCurve(
     FloatCurve=DataInterface() {propertyValues {Curve {keys [
         {interpMode=RCIM_Cubic, tangentMode=RCTM_User, time=0, value=0}
         <at 0.25 1>
         {interpMode=RCIM_Cubic, leaveTangent=-1.0256, time=0.7, value=1}
        ]}}}
     Scale Curve=$User.OpacityMult
    )) {isSetParameters=true}
   /SetVariables(Particles.DecalSize:Vector3f=&Multiply_VectorByFloat(
     Vector=$Particles.Initial.DecalSize
     Float=&FloatFromCurve(
      FloatCurve=DataInterface() {propertyValues {Curve {keys [{arriveTangent=139.099, interpMode=RCIM_Cubic, leaveTangent=139.099, tangentMode=RCTM_User, time=0, value=0}, <at 0.015 1 RCIM_Cubic>]}}}
      Scale Curve=1.5
     )
    )) {isSetParameters=true}
   &DynamicMaterialParameters(
    Index 0 Param 1=2
    Index 0 Param 2=1
    Module.Index0Float0Write=true
    Module.Index0Float1Write=true
    Module.Index0Float2Write=true
    Module.Index0Float3Write=true
   )
   /SetVariables() {isSetParameters=true}
   /SetVariables(Particles.groundglowpos:NiagaraPosition=&AddVectorToPosition(Position=$Particles.Initial.Position, Vector=<0 0 1>)) {isSetParameters=true}
   /SetVariables(Particles.groundglowsize:Vector2f=&Multiply_Vector2D_ByFloat(Float=$User.WidthMult, Vector2D=<150 150>)) {isSetParameters=true}
   /SetVariables(Particles.Vector 2D:Vector2f=&Multiply_Vector2D_ByFloat(Float=0.7, Vector2D=$Particles.groundglowsize)) {isSetParameters=true}
  ]
  rendererClasses [&NiagaraSpriteRendererProperties, &NiagaraDecalRendererProperties]
  renderers [
   {rendererClass=&NiagaraSpriteRendererProperties}
   {rendererClass=&NiagaraDecalRendererProperties, rendererIndex=1}
   {rendererClass=&NiagaraSpriteRendererProperties, rendererIndex=2}
   {rendererClass=&NiagaraSpriteRendererProperties, rendererIndex=3}
  ]
 }}
summary {
 systemName=NS_Emote_TulipGlory_GroundCracks
 emitters [{emitterName=Ground_Cracks, rendererClasses [&NiagaraSpriteRendererProperties, &NiagaraDecalRendererProperties]}]
 userVariables [
  User.bIsFrontEnd {type=NiagaraBool}
  User.bIsFrontEndPreview {type=NiagaraBool}
  User.LengthMult {type=NiagaraFloat, defaultValue=1, description=""}
  User.LengthMultAnimated {type=NiagaraFloat, defaultValue=1, description=""}
  User.OpacityMult {type=NiagaraFloat, defaultValue=1, description=""}
  User.RibbonIDOffset {type=NiagaraInt32}
  User.WidthMult {type=NiagaraFloat, defaultValue=1, description=""}
 ]
}
