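# Niagara one-shot burst effect for the TulipGlory emote, named for the moment its prop spawns.
# Four short-lived sprite emitters layer a growing center puff, a dark glow with flash flare, directional streaks, and gravity-pulled drops.
# Each has two sprite renderers, with secondary size, color or rotation attributes computed on some emitters.
# Time-of-day functions tint or fade layers, System.GlobalScale drives size and drop speed, and front-end modifiers adjust the glow.
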NiagaraSystem /BRCosmetics/Animation/Game/MainPlayer/Emotes/TulipGlory/FX/NS_Emote_TulipGlory_PropSpawn
dependencies {
 usedDataInterfaces [
  &NiagaraDataInterfaceCurve
  &NiagaraDataInterfaceVector2DCurve
  &NiagaraDataInterfaceArrayFloat3
 ]
 usedDynamicInputs [
  &SimulationPosition
  /Niagara/DynamicInputs/Math/Clamp_Vector
  &FN_FrontEndVector3Modifier
  &SelectVectorFromArray
 ]
 usedModules [
  &self:SystemSpawnScript.SystemScriptSource.SystemScriptGraph.NiagaraNodeAssignment_0.SetVariables_#1
  &SystemState
  &self:SystemSpawnScript.SystemScriptSource.SystemScriptGraph.NiagaraNodeAssignment_1.SetVariables_#2
 ]
}
emitters {
 Center002 {
  emitterUpdateScript [
   &EmitterState()
   &SpawnBurst_Instantaneous(Use Spawn Probability=false, UseLoopCountLimit=false)
   /SetVariables(Emitter.EmitterRandom:NiagaraFloat=&RandomRangeFloat(Minimum=0, Maximum=1)) {isSetParameters=true}
  ]
  particleSpawnScript [
   &InitializeParticle(
    Color=<rgba 0.8 0.622 0.575 1>
    Lifetime Max=0.25
    Lifetime Min=0.2
    Lifetime Mode=::NewEnumerator1
    Position Offset=<0 0 15>
    Sprite Rotation Angle=&RandomRangeFloat(Minimum=45, Maximum=125)
    Sprite Rotation Mode=::NewEnumerator0
    Sprite Size Mode=::NewEnumerator1
    Uniform Sprite Size Max=20
    Uniform Sprite Size Min=15
    UsePositionOffset=true
    Write Material Random=true
   )
   &CameraOffset(Camera Offset Amount=50)
  ]
  particleUpdateScript [
   &ParticleState()
   &DynamicMaterialParameters(
    Index 0 Param 1=$Emitter.EmitterRandom
    Index 0 Param 3=&FloatFromCurve(FloatCurve=DataInterface() {propertyValues {Curve {keys [
         {interpMode=RCIM_Cubic, tangentMode=RCTM_Break, time=0.3, value=-0.6}
         {interpMode=RCIM_Cubic, tangentMode=RCTM_Break, time=0.95, value=-1.3}
        ]}}})
    Index 0 Param 4=&FloatFromCurve(FloatCurve=DataInterface() {propertyValues {Curve {keys [{arriveTangent=2.6651, interpMode=RCIM_Cubic, leaveTangent=2.6651, tangentMode=RCTM_User, time=0, value=0.1}, <at 0.6 1.333 RCIM_Cubic>]}}})
    Module.Param0WriteEnabled=true
    Module.Param1WriteEnabled=true
    Module.Param2WriteEnabled=true
    Module.Param3WriteEnabled=true
   ) {module=DynamicMaterialParameters001}
   &GravityForce(Gravity=<0 0 -300>)
   &SolveForcesAndVelocity(Clamp Velocity=false, Limit Acceleration=false)
   &ScaleColor(
    Scale Alpha=&FloatFromCurve(FloatCurve=DataInterface() {propertyValues {Curve {keys [{interpMode=RCIM_Cubic, tangentMode=RCTM_User, time=0.7502, value=1}, {interpMode=RCIM_Cubic, tangentMode=RCTM_User, time=1, value=0}]}}})
    Scale RGB=&FN_Select_Vector_from_Time_Of_Day(Morning=<1 1 1>, Day=<1 1 1>, Evening=<1 1 1>, Night=<0.5 1 1.5>)
    ScaleA=true
    ScaleRGB=true
   )
   &ScaleSpriteSize(
    Uniform Curve Scale=&Multiply_Float(A=1.65, B=$System.GlobalScale)
    Uniform Curve Sprite Scale=DataInterface() {propertyValues {Curve {keys [
        {interpMode=RCIM_Cubic, leaveTangent=6.6646, tangentMode=RCTM_Break, time=0, value=0}
        {arriveTangent=3.357, interpMode=RCIM_Cubic, leaveTangent=3.357, time=0.4356, value=2.5}
        {interpMode=RCIM_Cubic, leaveTangent=0.9481, time=0.9983, value=3.3515}
       ]}}}
   )
   /SetVariables(
    Particles.Sprite Rotation 2:NiagaraFloat=&RandomRangeFloat(Minimum=-48, Maximum=-125)
    Particles.Sprite Size 2:Vector2f=&Multiply_Vector2D_ByFloat(Float=0.8, Vector2D=$Particles.SpriteSize)
   ) {isSetParameters=true}
  ]
  rendererClasses [&NiagaraSpriteRendererProperties]
  renderers [
   {rendererClass=&NiagaraSpriteRendererProperties}
   {rendererClass=&NiagaraSpriteRendererProperties, rendererIndex=1}
  ]
 }
 Dark_Glow {
  emitterUpdateScript [
   &EmitterState()
   &SpawnBurst_Instantaneous(Spawn Count=1, Use Spawn Probability=false)
  ]
  particleSpawnScript [
   &InitializeParticle(
    Color=<rgba 0.0451 0.0469 0.0286 1>
    Lifetime=0.3
    Position Offset=<0 0 15>
    Sprite Size Mode=::NewEnumerator3
    Uniform Sprite Size=&Multiply_Float(
     A=150
     B=&MakeCustomFloatFromBool(Bool=$User.bIsFrontEnd, True Float=&Multiply_Float(A=$User.FrontendScaler, B=1), False Float=1)
    )
    UsePositionOffset=true
    Write Material Random=true
   ) {module=InitializeParticle001}
   &DynamicMaterialParameters(
    Index 0 Param 1=1
    Index 0 Param 2=1
    Index 0 Param 3=1
    Index 0 Param 4=1
    Module.Param0WriteEnabled=true
    Module.Param1WriteEnabled=true
    Module.Param2WriteEnabled=true
    Module.Param3WriteEnabled=true
   )
   /SetVariables(
    Particles.Flash Flare CamOff=20
    Particles.Flash Flare Color:LinearColor=<rgba 0.0451 0.0469 0.0286 1>
    Particles.Flash Flare Lifetime:NiagaraFloat=&FN_FrontEndFloatModifier(Gameplay=0.2, Front End=0.125)
    Particles.Flash Flare Sprite Size:Vector2f=&Multiply_Vector2D_ByFloat(
     Float=1
     Vector2D=&FN_FrontEndVector2Modifier(Front End:Vector2f=<120 120>, Gameplay:Vector2f=<150 150>)
    )
   ) {isSetParameters=true}
   &CameraOffset(Camera Offset Amount=60)
  ]
  particleUpdateScript [
   &ParticleState()
   &SolveForcesAndVelocity(Clamp Velocity=false, Limit Acceleration=false)
   &ScaleColor(
    Scale Alpha=&Multiply_Float(
     A=&FloatFromCurve(
      FloatCurve=DataInterface() {propertyValues {Curve {keys [<at 0 1>,<at 1 0>]}}}
      Scale Curve=&FN_FrontEndFloatModifier(Gameplay=2.6, Front End=0.5)
     )
     B=&FNTimeOfDayScalar(Night Scale=0.3, Morning Scale=0.05, Day Scale=0.5, Evening Scale=0.05)
    )
    Scale RGB=<0 0 0>
    ScaleA=true
    ScaleRGB=true
   )
   &ScaleSpriteSize(Scale Factor=&Vector2DFromFloat(Value=&FloatFromCurve(FloatCurve=DataInterface() {propertyValues {Curve {keys [<at 0 1>,<at 1 3>]}}}, Scale Curve=$System.GlobalScale)))
   /SetVariables(Particles.Update Flash Flare Age:NiagaraFloat=&Divide_Float(A=$Particles.Age, B=$Particles.Initial.Flash Flare Lifetime)) {isSetParameters=true}
   /SetVariables(Particles.Update Flash Flare Color:LinearColor=&MakeLinearColorFromVectorAndFloat(
     "Vector (RGB)"=&Multiply_Vector(
      A=&Multiply_Vector(
       A=&MakeVectorFromLinearColorRGB(Color=$Particles.Initial.Flash Flare Color)
       B=&VectorFromFloat(Value=&FN_FrontEndFloatModifier(Gameplay=1, Front End=1))
      )
      B=&FN_Select_Vector_from_Time_Of_Day(Morning=<1 1 1>, Day=<1 1 1>, Evening=<1 1 1>, Night=<0.6 0.7 1.2>)
     )
     "Float (Alpha)"=&FloatFromCurve(
      FloatCurve=DataInterface() {propertyValues {Curve {
         defaultValue=1
         keys [
          {arriveTangent=4.2516, interpMode=RCIM_Cubic, leaveTangent=4.2516, tangentMode=RCTM_User, time=0, value=0}
          {arriveTangent=0.1532, interpMode=RCIM_Cubic, leaveTangent=0.1532, tangentMode=RCTM_User, time=0.25, value=1}
          {arriveTangent=-2.96, tangentMode=RCTM_User, time=1, value=0}
         ]
        }}}
      CurveIndex=$Particles.Update Flash Flare Age
     )
    )) {isSetParameters=true}
  ]
  rendererClasses [&NiagaraSpriteRendererProperties]
  renderers [
   {rendererClass=&NiagaraSpriteRendererProperties}
   {rendererClass=&NiagaraSpriteRendererProperties, rendererIndex=1}
  ]
 }
 Directional_Streaks {
  emitterUpdateScript [
   &EmitterState()
   &SpawnBurst_Instantaneous(Spawn Count=5, Use Spawn Probability=true, UseLoopCountLimit=false)
  ]
  particleSpawnScript [
   &InitializeParticle(
    Color=&FN_Select_Color_from_Time_Of_Day(Morning=<rgba 0.0385 0.04 0.0244 1>, Day=<rgba 0.0385 0.04 0.0244 1>, Evening=<rgba 0.0385 0.04 0.0244 1>, Night=<rgba 0.0244 0.0306 0.04 1>)
    Lifetime Max=0.3
    Lifetime Min=0.25
    Lifetime Mode=::NewEnumerator1
    Position Offset=<0 0 15>
    Sprite Size=&Multiply_Vector2D_ByFloat(
     Float=$System.GlobalScale
     Vector2D=&Multiply_Vector2D(A=&RandomRangeVector2D(Minimum=<30 125>, Maximum=<40 150>), B=<1 1.65>)
    )
    Sprite Size Mode=::NewEnumerator4
    Sprite UV Mode=::NewEnumerator1
    UsePositionOffset=true
    Write Material Random=true
   ) {module=InitializeParticle001}
   &SubUVAnimation(End Frame=3, SubUV Animation Mode:ENiagara_SubUVLookupMode=::NewEnumerator2)
   /SetVariables(Particles.Color Dark Streak:LinearColor=<rgba 0.0096 0.01 0.0061 0.5>) {isSetParameters=true}
   &CameraOffset(Camera Offset Amount=70) {module=CameraOffset001}
   &AddVelocity(
    Module.Use Falloff Distance=false
    Origin Offset=<0 0 15>
    Origin Offset Coordinate Space=::Simulation
    Override Seed=false
    Velocity Mode=::NewEnumerator1
   ) {module=AddVelocity001}
  ]
  particleUpdateScript [
   &ParticleState()
   &SolveForcesAndVelocity(Clamp Velocity=false, Limit Acceleration=false)
   &ScaleSpriteSize(Scale Factor=&Vector2DFromCurve(
     Vector2Curve=DataInterface() {propertyValues {XCurve {keys [<at 0 1 RCIM_Cubic>,<at 1 1 RCIM_Cubic>]}, YCurve {keys [<at 0 0.5 RCIM_Cubic>,<at 1 1 RCIM_Cubic>]}}}
     ScaleCurve=&Multiply_Vector2D_ByFloat(Float=1, Vector2D=<1.5 1>)
    ))
   &ScaleColor(
    Scale Alpha=&FloatFromCurve(
     FloatCurve=DataInterface() {propertyValues {Curve {
        defaultValue=10
        keys [{interpMode=RCIM_Cubic, tangentMode=RCTM_User, time=0, value=1}, {interpMode=RCIM_Cubic, tangentMode=RCTM_User, time=1, value=0}]
       }}}
     Scale Curve=&FNTimeOfDayScalar(Night Scale=1, Morning Scale=1, Evening Scale=1)
    )
    Scale RGB=&FN_Select_Vector_from_Time_Of_Day(Morning=<1 1 1>, Day=<1 1 1>, Evening=<1 1 1>, Night=<1 1 1>)
    ScaleA=true
    ScaleRGB=true
   )
   /SetVariables(
    Particles.Update Color Dark Streak:LinearColor=&MultiplyLinearColor(
     Linear Color=$Particles.Color Dark Streak
     Scale Factor=&MakeVector4(
      X=1
      Y=1
      Z=1
      W=&FloatFromCurve(
       FloatCurve=DataInterface() {propertyValues {Curve {
          defaultValue=1
          keys [{interpMode=RCIM_Cubic, tangentMode=RCTM_User, time=0, value=1}, {interpMode=RCIM_Cubic, tangentMode=RCTM_User, time=1, value=0}]
         }}}
       Scale Curve=&Multiply_Float(A=2, B=&FNTimeOfDayScalar(Night Scale=0.8, Morning Scale=1, Evening Scale=1))
      )
     )
    )
    Particles.Update Size Dark Streak:Vector2f=&Multiply_Vector2D(A=$Particles.SpriteSize, B=<2 1.65>)
   ) {isSetParameters=true}
   &DynamicMaterialParameters(
    Index 0 Param 1=&FloatFromCurve(FloatCurve=DataInterface() {propertyValues {Curve {
        defaultValue=1
        keys [
         <at 0 1 RCIM_Cubic>
         {interpMode=RCIM_Cubic, tangentMode=RCTM_User, time=0.2, value=0.25}
         {interpMode=RCIM_Cubic, tangentMode=RCTM_User, time=1, value=0.8}
        ]
       }}})
    Module.Index0Float0Write=true
    Module.Index0Float1Write=true
    Module.Index0Float2Write=true
    Module.Index0Float3Write=true
   ) {module=DynamicMaterialParameters001}
  ]
  rendererClasses [&NiagaraSpriteRendererProperties]
  renderers [
   {rendererClass=&NiagaraSpriteRendererProperties}
   {rendererClass=&NiagaraSpriteRendererProperties, rendererIndex=1}
  ]
 }
 Drops001 {
  emitterUpdateScript [
   &EmitterState()
   &SpawnBurst_Instantaneous(Spawn Count=&RandomRangeInt(Minimum=15, Maximum=20), Use Spawn Probability=false)
  ]
  particleSpawnScript [
   &InitializeParticle(
    Color=<rgba 0.8 0.8 0.8 1>
    Lifetime=&UniformRangedFloat(Minimum=0.4, Maximum=0.8)
    Mass Mode=::NewEnumerator1
    Position Offset=<0 0 15>
    Sprite Size=&Multiply_Vector2D_ByFloat(
     Float=&UniformAOrBFloat(
      A=&RandomRangeFloat(Minimum=2.5, Maximum=6)
      B=7
      Distribution Weight=0.1
      Evaluation Type=::NewEnumerator0
      Recalculate Random Each Loop=true
      Override Seed=false
     )
     Vector2D=&UniformRangedVector2D(Minimum=<2 3>, Maximum=<2 4>)
    )
    Sprite Size Mode=::NewEnumerator4
    UsePositionOffset=true
    Write Material Random=true
   ) {module=InitializeParticle001}
   /Niagara/Modules/Spawn/Location/SphereLocation(MaskSpawnGroup=false, Module.Override Randomness=false, Sphere Radius=5, Surface Only=false)
   &AddVelocityFromPoint(
    Module.Use Falloff Distance=false
    Origin Offset=<0 0 15>
    Velocity Strength=&Multiply_Float(A=&UniformRangedFloat(Minimum=333, Maximum=444), B=$System.GlobalScale)
   )
   /Niagara/Modules/Update/SubUV/SubUVAnimation(SubUV Animation Mode=::NewEnumerator1) {module=SubUVAnimation001}
   /SetVariables(Particles.VisibilityTag=&RandomRangeInt(Minimum=0, Maximum=1)) {isSetParameters=true}
   &CameraOffset(Camera Offset Amount=50)
  ]
  particleUpdateScript [
   &ParticleState()
   &Drag(Drag=2.5, Use Linear Drag=true, Use Rotational Drag=true)
   &GravityForce(Gravity=&MakeVector(X=0, Y=0, Z=&RandomRangeFloat(Minimum=-400, Maximum=-500))) {module=GravityForce001}
   &SolveForcesAndVelocity(Clamp Velocity=false, Limit Acceleration=false)
   &ScaleSpriteSize(Scale Factor=&Multiply_Vector2D_ByFloat(
     Float=&FloatFromCurve(FloatCurve=DataInterface() {propertyValues {Curve {keys [<at 0 0.5>,<at 1 1>]}}}, Scale Curve=$System.GlobalScale)
     Vector2D=<0.8 0.8>
    ))
   &ScaleColor(
    Scale Alpha=&FloatFromCurve(FloatCurve=DataInterface() {propertyValues {Curve {keys [<at 0 1>,<at 0.75 1>,<at 1 0>]}}})
    Scale RGB=&FN_Select_Vector_from_Time_Of_Day(Morning=<1 1 1>, Day=<1 1 1>, Evening=<1 1 1>, Night=<1 1 1.5>)
    ScaleA=true
    ScaleRGB=true
   )
   &DynamicMaterialParameters(
    Index 0 Param 1=&RandomRangeFloat(Minimum=0, Maximum=1)
    Index 0 Param 2=&FloatFromCurve(FloatCurve=DataInterface() {propertyValues {Curve {keys [<at 0.35 0.45>,<at 1 0.9>]}}})
    Index 0 Param 3=&FloatFromCurve(FloatCurve=DataInterface() {propertyValues {Curve {keys [{arriveTangent=2.6651, interpMode=RCIM_Cubic, leaveTangent=2.6651, tangentMode=RCTM_User, time=0, value=0}, <at 1 1 RCIM_Cubic>]}}})
    Index 0 Param 4=&FloatFromCurve(FloatCurve=DataInterface() {propertyValues {Curve {keys [<at 0 0 RCIM_Cubic>, {arriveTangent=2.332, interpMode=RCIM_Cubic, leaveTangent=2.332, tangentMode=RCTM_User, time=1, value=1}]}}})
    Module.Param0WriteEnabled=true
    Module.Param1WriteEnabled=true
    Module.Param2WriteEnabled=false
    Module.Param3WriteEnabled=true
   ) {module=DynamicMaterialParameters001}
  ]
  rendererClasses [&NiagaraSpriteRendererProperties]
  renderers [
   {rendererClass=&NiagaraSpriteRendererProperties}
   {rendererClass=&NiagaraSpriteRendererProperties, rendererIndex=1}
  ]
 }
}
summary {
 systemName=NS_Emote_TulipGlory_PropSpawn
 emitters [
  {emitterName=Center002, rendererClasses [&NiagaraSpriteRendererProperties]}
  {emitterName=Dark_Glow, rendererClasses [&NiagaraSpriteRendererProperties]}
  {emitterName=Drops001, rendererClasses [&NiagaraSpriteRendererProperties]}
  {emitterName=Directional_Streaks, rendererClasses [&NiagaraSpriteRendererProperties]}
 ]
 userVariables [
  User.bIsFrontEnd {type=NiagaraBool}
  User.bIsFrontEndPreview {type=NiagaraBool}
  User.FrontendScaler {type=NiagaraFloat, defaultValue=0.555, description=""}
  User.SwingSectionIndex {type=NiagaraInt32}
 ]
}
