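# Niagara system for the Tulip Glory emote's steam, combining rising smoke, mesh trails and ground cracks.
# Smoke sprites spawn continuously, climbing via per-update position offsets, with time-of-day tint shared by the trails.
# A burst Ground_Cracks emitter drives a decal plus two sprite renderers; mesh Trails are shaped through dynamic material parameters.
# User width, length and opacity multipliers scale the layers, while front-end modifiers adjust camera offset and crack opacity.
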NiagaraSystem /BRCosmetics/Animation/Game/MainPlayer/Emotes/TulipGlory/FX/NS_Emote_TulipGlory_Steam
dependencies {
 usedDataInterfaces [&NiagaraDataInterfaceCurve, &NiagaraDataInterfaceSkeletalMesh]
 usedDynamicInputs []
 usedModules [
  &self:SystemSpawnScript.SystemScriptSource.SystemScriptGraph.NiagaraNodeAssignment_0.SetVariables_#1
  &SystemState
 ]
}
emitters {
 FN_Core_Smoke_Sprite001 {
  emitterUpdateScript [
   &EmitterState()
   &SpawnRate(SpawnRate=8, Use Spawn Probability=false)
  ]
  particleSpawnScript [
   /SetVariables(Particles.InitSpriteRotRate=&RandomRangeFloat(Minimum=-2, Maximum=2)) {isSetParameters=true}
   &InitializeParticle(
    Color=&FN_Select_Color_from_Time_Of_Day(
     Morning=<rgba 0.4479 0.4479 0.4479 1>
     Day=<rgba 0.4479 0.4479 0.4479 1>
     Evening=<rgba 0.4479 0.4479 0.4479 1>
     Night=<rgba 0.147 0.2283 0.4479 1>
    )
    Lifetime Max=4.5
    Lifetime Min=4
    Lifetime Mode=::NewEnumerator1
    Sprite Rotation Angle=&RandomRangeFloat(Minimum=0, Maximum=360)
    Sprite Rotation Mode=::NewEnumerator0
    Sprite Size Mode=::NewEnumerator3
    Uniform Sprite Size=&Multiply_Float(
     A=&RandomRangeFloat(Minimum=30, Maximum=40)
     B=&Multiply_Float(A=$User.WidthMult, B=0.8)
    )
    UsePositionOffset=false
    Write Material Random=true
   )
   &CameraOffset(Camera Offset Amount=&FN_FrontEndFloatModifier(Gameplay=0, Front End=&RandomRangeFloat(Minimum=-2, Maximum=2)))
  ]
  particleUpdateScript [
   &ParticleState()
   &SpriteRotationRate(Rotation Rate=&Multiply_Float(
     A=$Particles.InitSpriteRotRate
     B=&FloatFromCurve(
      FloatCurve=DataInterface() {propertyValues {Curve {keys [
          {arriveTangent=-4.0107, interpMode=RCIM_Cubic, leaveTangent=-4.0107, tangentMode=RCTM_User, time=-0.0046, value=1.0113}
          {arriveTangent=-0.2496, interpMode=RCIM_Cubic, leaveTangent=-0.2496, tangentMode=RCTM_User, time=0.5, value=0.16}
         ]}}}
      CurveIndex=$Emitter.Age
      Scale Curve=150
     )
    ))
   &ScaleColor(
    Scale Alpha=&FloatFromCurve(
     FloatCurve=DataInterface() {propertyValues {Curve {keys [
         {interpMode=RCIM_Cubic, tangentMode=RCTM_User, time=0, value=0}
         {arriveTangent=1.0396, interpMode=RCIM_Cubic, leaveTangent=1.0396, time=0.2, value=1}
         {arriveTangent=-1.25, interpMode=RCIM_Cubic, leaveTangent=-1.25, time=0.7576, value=0.7876}
         <at 1 0>
        ]}}}
     Scale Curve=&Multiply_Float(A=0.025, B=$User.OpacityMult)
    )
    Scale RGB=&VectorFromFloat(Value=&FloatFromCurve(FloatCurve=DataInterface() {propertyValues {Curve {keys [{arriveTangent=1.8145, interpMode=RCIM_Cubic, leaveTangent=4.5795, tangentMode=RCTM_Break, time=0, value=0.4}, <at 0.3387 0.9565>, <at 1 1>]}}}))
    ScaleA=true
    ScaleRGB=true
   )
   &ScaleSpriteSize(Scale Factor=&Vector2DFromFloat(Value=&FloatFromCurve(FloatCurve=DataInterface() {propertyValues {Curve {keys [
          {arriveTangent=1.6485, interpMode=RCIM_Cubic, leaveTangent=1.6485, tangentMode=RCTM_User, time=0, value=0.25}
          {arriveTangent=0.3967, interpMode=RCIM_Cubic, leaveTangent=0.3967, tangentMode=RCTM_User, time=0.3204, value=0.8446}
          {arriveTangent=0.1638, interpMode=RCIM_Cubic, leaveTangent=0.1638, tangentMode=RCTM_User, time=1.0463, value=1}
         ]}}})))
   &DynamicMaterialParameters(
    Index 0 Param 1=&Add_Float(
     A=$Particles.RandomFloat
     B=&FloatFromCurve(FloatCurve=DataInterface() {propertyValues {Curve {keys [
          {arriveTangent=2.7533, interpMode=RCIM_Cubic, leaveTangent=1.3017, tangentMode=RCTM_Break, time=0, value=0}
          {arriveTangent=0.0529, interpMode=RCIM_Cubic, leaveTangent=0.2021, tangentMode=RCTM_User, time=1, value=0.5}
         ]}}})
    )
    Index 0 Param 3=&FloatFromCurve(FloatCurve=DataInterface() {propertyValues {Curve {keys [{arriveTangent=-53.9199, time=0, value=0}, {leaveTangent=0.2723, time=0.25, value=0}]}}})
    Index 0 Param 4=&FloatFromCurve(FloatCurve=DataInterface() {propertyValues {Curve {keys [<at 0 1>,<at 1 0>]}}})
    Index 1 Param 1=&FloatFromCurve(FloatCurve=DataInterface() {propertyValues {Curve {keys [<at 0 0>, {arriveTangent=4.2257, interpMode=RCIM_Cubic, leaveTangent=4.2257, tangentMode=RCTM_User, time=0.25, value=0}]}}})
    Index 1 Param 2=&FloatFromCurve(FloatCurve=DataInterface() {propertyValues {Curve {keys [<at 0 0>]}}})
    Index 1 Param 3=&FloatFromCurve(FloatCurve=DataInterface() {propertyValues {Curve {keys [{interpMode=RCIM_Cubic, leaveTangent=6.3917, tangentMode=RCTM_Break, time=0, value=0}]}}})
    Index 1 Param 4=&FloatFromCurve(FloatCurve=DataInterface() {propertyValues {Curve {keys [{arriveTangent=-0.136, leaveTangent=-0.136, time=0, value=1}]}}})
    Index 2 Param 1=&FloatFromCurve(FloatCurve=DataInterface() {propertyValues {Curve {keys [
         {arriveTangent=2.646, interpMode=RCIM_Cubic, leaveTangent=2.646, tangentMode=RCTM_User, time=0, value=0}
         {leaveTangent=-0.0719, time=1, value=1}
        ]}}})
    Index 2 Param 2=&FloatFromCurve(
     FloatCurve=DataInterface() {propertyValues {Curve {keys [
         {arriveTangent=1.9056, time=0, value=0.2}
         {arriveTangent=0.0259, interpMode=RCIM_Cubic, leaveTangent=0.0259, tangentMode=RCTM_User, time=1, value=1.4}
        ]}}}
     Scale Curve=&RandomRangeFloat(Minimum=0.5, Maximum=1)
    )
    Index 2 Param 3=&FloatFromCurve(FloatCurve=DataInterface() {propertyValues {Curve {keys [
         {interpMode=RCIM_Cubic, leaveTangent=2.3064, tangentMode=RCTM_Break, time=0, value=1}
         {interpMode=RCIM_Cubic, leaveTangent=11.1233, tangentMode=RCTM_Break, time=1, value=2}
        ]}}})
    Index 2 Param 4=&FloatFromCurve(
     FloatCurve=DataInterface() {propertyValues {Curve {keys [
         {arriveTangent=-0.6867, interpMode=RCIM_Cubic, leaveTangent=-0.6867, tangentMode=RCTM_User, time=0, value=1}
         {interpMode=RCIM_Cubic, leaveTangent=-1.2956, tangentMode=RCTM_Break, time=1, value=0.75}
        ]}}}
     Scale Curve=1.6
    )
    Module.Param0WriteEnabled=true
    Module.Param0WriteEnabled1=true
    Module.Param0WriteEnabled2=true
    Module.Param1WriteEnabled=false
    Module.Param1WriteEnabled1=true
    Module.Param1WriteEnabled2=true
    Module.Param2WriteEnabled=true
    Module.Param2WriteEnabled1=true
    Module.Param2WriteEnabled2=true
    Module.Param3WriteEnabled=true
    Module.Param3WriteEnabled1=true
    Module.Param3WriteEnabled2=true
    Write Parameter Index 1=true
    Write Parameter Index 2=true
   )
   /SetVariables(Particles.Position=&AddVectorToPosition(
     Position=$Particles.Position
     Vector=&MakeVector(
      X=0
      Y=0
      Z=&Multiply_Float(
       A=&RandomRangeFloat(Minimum=0.8, Maximum=1)
       B=&Multiply_Float(A=$User.LengthMult, B=$User.LengthMultAnimated)
      )
     )
    )) {isSetParameters=true}
   &SubUVAnimation(Start Frame:NiagaraInt32=&RandomRangeInt(Minimum=0, Maximum=10))
  ]
  rendererClasses [&NiagaraSpriteRendererProperties]
  renderers [{rendererClass=&NiagaraSpriteRendererProperties}]
 }
 Ground_Cracks {
  emitterUpdateScript [
   &EmitterState(
    Inactive Response=::NewEnumerator1
    Life Cycle Mode=::NewEnumerator1
    Loop Behavior=::NewEnumerator1
    Loop Delay=0.5
    Loop Duration Mode=::NewEnumerator1
    UseLoopDelay=false
   )
   &SpawnBurst_Instantaneous(Use Spawn Probability=false, UseLoopCountLimit=false)
  ]
  particleSpawnScript [
   &InitializeParticle(
    Color=<rgba 0.0104 0.0104 0.0104 1>
    Lifetime=4.5
    Sprite Rotation Mode=::NewEnumerator2
    Sprite Size Mode=::NewEnumerator1
    Uniform Sprite Size Max=130
    Uniform Sprite Size Min=90
    UsePositionOffset=true
    Write Material Random=true
   ) {module=InitializeParticle001}
   /SetVariables(Particles.RandomDecalYZ:NiagaraFloat=&Multiply_Float(A=&RandomRangeFloat(Minimum=60, Maximum=80), B=$User.WidthMult)) {isSetParameters=true}
   /SetVariables(Particles.DecalSize:Vector3f=&MakeVector(X=75, Y=$Particles.RandomDecalYZ, Z=$Particles.RandomDecalYZ)) {isSetParameters=true}
   /SetVariables(Particles.VisibilityTag=&MakeCustomIntFromBool(Bool=$User.bIsFrontEndPreview, True INT=0, False INT=1)) {isSetParameters=true}
   &SpriteFacingAndAlignment(Sprite Alignment=<1 0 0>, Sprite Facing=<0 0 1>, WriteAlignment=true, WriteFacing=true) {module=SpriteFacingAndAlignment001}
   /SetVariables(Particles.DecalOrientation:Quat4f=&MultiplyQuaternion(
     Quaternion A=&MakeQuaternion(
      Quaternion From=0
      Axis=<0 0 1>
      Angle=&RandomRangeFloat(Minimum=0, Maximum=360)
      From=?
      To=?
      Angle Type=::NewEnumerator0
      Coordinate Space=::World
      Normalize Vectors=false
     )
     Quaternion B=&MakeQuaternion(Quaternion From=0, Axis=<0 1 0>, Angle=90, From=?, To=?, Angle Type=::NewEnumerator0, Coordinate Space=::World, Normalize Vectors=false)
    )) {isSetParameters=true}
   &CameraOffset(Camera Offset Amount=&FN_FrontEndFloatModifier(Gameplay=0, Front End=5))
  ]
  particleUpdateScript [
   &ParticleState(Kill Particles When Lifetime Has Elapsed=false, Let Infinitely Lived Particles Die When Emitter Deactivates=true)
   &ScaleSpriteSize(Uniform Curve Sprite Scale=DataInterface() {propertyValues {Curve {keys [
        {arriveTangent=230.8238, interpMode=RCIM_Cubic, leaveTangent=230.8238, tangentMode=RCTM_User, time=0, value=0}
        {arriveTangent=0.8135, interpMode=RCIM_Cubic, leaveTangent=0.8135, tangentMode=RCTM_User, time=0.01, value=1}
       ]}}})
   &ScaleColor(
    Scale Alpha=&FloatFromCurve(
     FloatCurve=DataInterface() {propertyValues {Curve {keys [
         {interpMode=RCIM_Cubic, tangentMode=RCTM_User, time=0, value=0}
         <at 0.5 1>
         {interpMode=RCIM_Cubic, leaveTangent=-1.0256, time=0.7, value=1}
        ]}}}
     Scale Curve=&Multiply_Float(A=$User.OpacityMult, B=&FN_FrontEndFloatModifier(Gameplay=0.8, Front End=0.4))
    )
    ScaleA=true
    ScaleRGB=true
   )
   &ScaleSpriteSize(Uniform Curve Sprite Scale=DataInterface() {propertyValues {Curve {keys [{arriveTangent=139.099, interpMode=RCIM_Cubic, leaveTangent=139.099, tangentMode=RCTM_User, time=0, value=0}, <at 0.015 1 RCIM_Cubic>]}}}) {module=ScaleSpriteSize001}
   /SetVariables(Particles.DecalFade:NiagaraFloat=&FloatFromCurve(
     FloatCurve=DataInterface() {propertyValues {Curve {keys [
         {interpMode=RCIM_Cubic, tangentMode=RCTM_User, time=0, value=0}
         <at 0.25 1>
         {interpMode=RCIM_Cubic, leaveTangent=-1.0256, time=0.7, value=1}
        ]}}}
     Scale Curve=$User.OpacityMult
    )) {isSetParameters=true}
   /SetVariables(Particles.DecalSize:Vector3f=&Multiply_VectorByFloat(
     Vector=$Particles.Initial.DecalSize
     Float=&FloatFromCurve(FloatCurve=DataInterface() {propertyValues {Curve {keys [{arriveTangent=139.099, interpMode=RCIM_Cubic, leaveTangent=139.099, tangentMode=RCTM_User, time=0, value=0}, <at 0.015 1 RCIM_Cubic>]}}})
    )) {isSetParameters=true}
   &DynamicMaterialParameters(
    Index 0 Param 1=2
    Module.Index0Float0Write=true
    Module.Index0Float1Write=true
    Module.Index0Float2Write=true
    Module.Index0Float3Write=true
   )
   /SetVariables() {isSetParameters=true}
   /SetVariables(Particles.groundglowpos:NiagaraPosition=&AddVectorToPosition(Position=$Particles.Initial.Position, Vector=<0 0 1>)) {isSetParameters=true}
   /SetVariables(Particles.groundglowsize:Vector2f=&Multiply_Vector2D_ByFloat(Float=$User.WidthMult, Vector2D=<100 100>)) {isSetParameters=true}
   /SetVariables(Particles.crackssize:Vector2f=&Multiply_Vector2D_ByFloat(Float=5.75, Vector2D=$Particles.groundglowsize)) {isSetParameters=true}
   /SetVariables(Particles.colorcrack:LinearColor=&MultiplyLinearColor(Linear Color=$Particles.Color, Scale Factor=<1 1 1 0.8>)) {isSetParameters=true}
  ]
  rendererClasses [&NiagaraDecalRendererProperties, &NiagaraSpriteRendererProperties]
  renderers [
   {rendererClass=&NiagaraDecalRendererProperties, rendererIndex=0}
   {rendererClass=&NiagaraSpriteRendererProperties, rendererIndex=1}
   {rendererClass=&NiagaraSpriteRendererProperties, rendererIndex=2}
  ]
 }
 Trails {
  emitterUpdateScript [
   &EmitterState(
    Inactive Response=::NewEnumerator1
    Life Cycle Mode=::NewEnumerator1
    Loop Behavior=::NewEnumerator1
    Loop Duration Mode=::NewEnumerator1
    UseLoopDelay=false
   )
   &SpawnBurst_Instantaneous(Use Spawn Probability=false, UseLoopCountLimit=false)
  ]
  particleSpawnScript [
   &FN_SetSpawnGroup(Number of Spawn Groups=3)
   &InitializeParticle(
    Color=&FN_Select_Color_from_Time_Of_Day(
     Morning=<rgba 0.9062 0.9062 0.9062 1>
     Day=<rgba 0.9062 0.9062 0.9062 1>
     Evening=<rgba 0.9062 0.9062 0.9062 1>
     Night=<rgba 0.2974 0.4748 0.9062 1>
    )
    Lifetime=3
    Mesh Scale=&Multiply_Vector(
     A=&RandomRangeVector(Minimum=<1 1 2.5>, Maximum=<1 1 3>)
     B=&MakeVector(X=1, Y=1, Z=$User.LengthMult)
    )
    Mesh Scale Mode=::NewEnumerator4
    UsePositionOffset=false
    Write Material Random=true
   )
   /SetVariables(Particles.rand:NiagaraFloat=&RandomRangeFloat(Minimum=0, Maximum=1)) {isSetParameters=true}
  ]
  particleUpdateScript [
   &ParticleState(Kill Particles When Lifetime Has Elapsed=false, Let Infinitely Lived Particles Die When Emitter Deactivates=true)
   &DynamicMaterialParameters(
    Index 0 Param 1=&Multiply_Float(A=&Lerp_Float(A=20, B=40, Alpha=$Particles.Initial.rand), B=$User.WidthMult)
    Index 0 Param 2=&Multiply_Float(A=&Lerp_Float(A=60, B=80, Alpha=$Particles.Initial.rand), B=$User.WidthMult)
    Index 0 Param 3=&RandomRangeFloat(Minimum=0.8, Maximum=1)
    Index 0 Param 4=&RandomRangeFloat(Minimum=0, Maximum=1)
    Index 1 Param 2=&FloatFromCurve(
     FloatCurve=DataInterface() {propertyValues {
       bOptimizeLUT=false
       bUseLUT=false
       Curve {defaultValue=1, keys [{tangentMode=RCTM_User, time=0, value=0}, {tangentMode=RCTM_User, time=5, value=1}]}
      }}
     CurveIndex=$Particles.Age
    )
    Index 1 Param 4=1
    Index 3 Param 1=&MakeFloatFromVector(Vector=$StackContext.InterpolateOverTime.InterpolateOverTime_V3_.Moving Average, Channel=::NewEnumerator0)
    Index 3 Param 2=&MakeFloatFromVector(Vector=$StackContext.InterpolateOverTime.InterpolateOverTime_V3_.Moving Average, Channel=::NewEnumerator1)
    Index 3 Param 3=&MakeFloatFromVector(Vector=$StackContext.InterpolateOverTime.InterpolateOverTime_V3_.Moving Average, Channel=::NewEnumerator2)
    Module.Index0Param0Write=true
    Module.Index0Param1Write=true
    Module.Index0Param2Write=true
    Module.Index0Param3Write=true
    Module.Index1Param0Write=true
    Module.Index1Param1Write=true
    Module.Index1Param2Write=true
    Module.Index1Param3Write=true
    Module.Index2Param0Write=true
    Module.Index2Param1Write=true
    Module.Index2Param2Write=false
    Module.Index2Param3Write=false
    Module.Index3Param0Write=true
    Module.Index3Param1Write=true
    Module.Index3Param2Write=true
    Module.Index3Param3Write=true
    Write Parameter Index 1=true
    Write Parameter Index 2=true
    Write Parameter Index 3=true
   )
   &ScaleColor(
    Scale Alpha=&FloatFromCurve(
     FloatCurve=DataInterface() {propertyValues {Curve {keys [<at 0 0 RCIM_Cubic>,<at 2 1 RCIM_Cubic>]}}}
     CurveIndex=$System.Age
     Scale Curve=&Multiply_Float(A=0.025, B=$User.OpacityMult)
    )
    ScaleA=true
    ScaleRGB=false
   )
  ]
  rendererClasses [&NiagaraMeshRendererProperties]
  renderers [{rendererClass=&NiagaraMeshRendererProperties, rendererIndex=0}]
 }
}
summary {
 systemName=NS_Emote_TulipGlory_Steam
 emitters [
  {emitterName=Trails, rendererClasses [&NiagaraMeshRendererProperties]}
  {emitterName=FN_Core_Smoke_Sprite001, rendererClasses [&NiagaraSpriteRendererProperties]}
  {emitterName=Ground_Cracks, rendererClasses [&NiagaraDecalRendererProperties, &NiagaraSpriteRendererProperties]}
 ]
 userVariables [
  User.bIsFrontEnd {type=NiagaraBool}
  User.bIsFrontEndPreview {type=NiagaraBool}
  User.LengthMult {type=NiagaraFloat, defaultValue=1, description=""}
  User.LengthMultAnimated {type=NiagaraFloat, defaultValue=1, description=""}
  User.OpacityMult {type=NiagaraFloat, defaultValue=1, description=""}
  User.RibbonIDOffset {type=NiagaraInt32}
  User.WidthMult {type=NiagaraFloat, defaultValue=1, description=""}
 ]
}
